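; Toy Fox boss ability that burrows the boss to a new reposition point, extending a parent burrow ability.
; Activation checks target proximity time, recent headbutt or stun tags, and frustration.
; An EQS query selects a reposition actor as the burrow destination, cancelling if the query is invalid.
; Effects and notifies manage stagger immunity and invulnerability while burrowed; unburrowing resets frustration, a delay precedes immunity removal, and ability end cleans up.
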(event Ability|EventActivateAbility
  ;[0.01369 0.125 0 0.5] ActivateAbility: Apply stagger immunity
  (|CallParentFunction)
  (bind _returnvalue (Ability|ApplyGameplayEffectToOwner (Variables|Settings|GetStaggerImmunityGameplayEffectClass)))
  (Variables|Transient|SetStaggerImmunityGameplayEffectHandle _returnvalue))

(event EventBeginRepositionQuery
  ;[0 0.15 0.01136 0.5] BeginRepositionQuery: Overridden to directly relocate only using EQS.
  (CallFunction|RelocateUsingEQS)
  (Fortnite|FortShippingLog self "BeginRepositionQuery: Invoked."))

(event EventRelocateUsingEQS
  ;[0.00861 0.05308 0.15 0.5] RelocateUsingEQS: Overridden to run with querier as avatar actor.
  (bind _returnvalue (AI|EQS|RunEQSQuery (Variables|Default|GetEQS_Repositioning) (Ability|GetAvatarActorfromActorInfo) (Variables|Default|GetEnvQueryRunMode)))
  (Utilities|IsValid _returnvalue
    (:Is%20Valid
      (AI|EQS|BindEventtoOnQueryFinished _returnvalue (EventDispatchers|CreateEvent :!event RepositioningEQSFinished)))
    (:Is%20Not%20Valid
      (Ability|CancelAbility)
      (Fortnite|FortShippingLog self "RelocateUsingEQS: Failed to create query." true))))

(event EventRepositioningEQSFinished
  (params (QueryInstance :type EQS%20Query%20Instance%20Object%20Reference) (QueryStatus :type EEnvQueryStatus%20Enum :default "Processing"))
  ;[0 0.15 0.06926 0.5] RepositionEQSFinished: Once EQS is finished, cast to the expected actor and being repositioning to that actor.
  (Utilities|IsValid QueryInstance
    (:Is%20Valid
      (bind _resultactors (AI|EQS|GetQueryResultsasActors QueryInstance))
      (if (Utilities|Array|IsValidIndex _resultactors)
        (bind _asb_tidal_crane_toy_fox_reposition_actor (Utilities|Casting|CastToB_TidalCraneToyFoxRepositionActor (Utilities|Array|Get(acopy) _resultactors))
          (:then
            (bind _output_get (Variables|Transient|SetSelectedRepositionActor _asb_tidal_crane_toy_fox_reposition_actor))
            (Variables|Default|SetNewLocation (Transformation|GetActorLocation _output_get))
            (Fortnite|FortShippingLog self (Utilities|String|Append "RepositionEQSFinished: Successfully found reposition actor: " (Utilities|GetDisplayName _output_get)))
            (CallFunction|StartBurrowSequence true))
          (:CastFailed
            _))
        (else
          _))
      (Fortnite|FortShippingLog self "RepositionEQSFinished: Failed to find EQS location to go to." true)
      (Ability|CancelAbility))))

(event EventTryKnockbackWithGameplayEffect
  ;[0 0.15 0.08092 0.5] Apply unburrow knockback ability to the boss.
  ; For a short duration, the boss knocks back everything around it.
  (Ability|ApplyGameplayEffectToOwner "/TidalCrane_ToyFox_Boss/Abilities/UnburrowKnockback/GE_TidalCrane_ToyFox_Boss_UnburrowKnockbackAbility.GE_TidalCrane_ToyFox_Boss_UnburrowKnockbackAbility_C"))

(event EventNotifyBegin
  (params (NotifyTag :type Gameplay%20Tag%20Structure))
  ;[0.1457 0.01317 0.15 0.5] CanBeDamaged Management: When burrowing is almost complete, disable can be damaged.
  ; Renable can be damaged when bug comes out of the ground again.
  (if (== NotifyTag "(TagName=\"Event.TidalCrane.Boss.BurrowingAlmostComplete\")")
    (bind _returnvalue (Ability|ApplyGameplayEffectToOwner (Variables|Settings|GetBurrowedGameplayEffect)))
    (Variables|Transient|SetBurrowedGameplayEffectHandle _returnvalue)
    (bind _toyfoxbossactor (Utilities|GetToyFoxBossActor)
      (:Valid
        (Fortnite|SetCanbeDamaged _toyfoxbossactor)))))

(event EventFullBurrowTravelGameplayComplete
  ;[0.1457 0.01317 0.15 0.5] CanBeDamaged Management: When burrowing is almost complete, disable can be damaged.
  ; Renable can be damaged when bug comes out of the ground again.
  (|CallParentFunction)
  (Utilities|FlowControl|Sequence
    (:then_0
      (Ability|RemoveGameplayEffectFromOwnerWithHandle (Variables|Transient|GetBurrowedGameplayEffectHandle))
      (bind _toyfoxbossactor (Utilities|GetToyFoxBossActor)
        (:Valid
          (Fortnite|SetCanbeDamaged _toyfoxbossactor true)
          (Frustration|ResetFrustration _toyfoxbossactor))))
    (:then_1
      (Utilities|FlowControl|Delay (ScalableFloat|GetCachedScalableFloatValue (Variables|Hotfix|GetHotfix_StaggerRemovePostUnburrowDelay)))
      (Ability|RemoveGameplayEffectFromOwnerWithHandle (Variables|Transient|GetStaggerImmunityGameplayEffectHandle)))))

(event Ability|EventOnEndAbility
  (params (bWasCancelled :type Boolean))
  ;[0.15 0.09655 0 0.5] EndPlay: Renable collision in case it isn't.
  (Utilities|FlowControl|Sequence
    (:then_0
      (Ability|RemoveGameplayEffectFromOwnerWithHandle (Variables|Transient|GetStaggerImmunityGameplayEffectHandle))
      (Ability|RemoveGameplayEffectFromOwnerWithHandle (Variables|Transient|GetBurrowedGameplayEffectHandle))
      (bind _toyfoxbossactor (Utilities|GetToyFoxBossActor)
        (:Valid
          (Fortnite|SetCanbeDamaged _toyfoxbossactor true))))
    (:then_1
      (|CallParentFunction bWasCancelled))))

(fn CanActivateAbility
  (params (ActorInfo :type Gameplay%20Ability%20Actor%20Info%20Structure) (Handle :type Gameplay%20Ability%20Spec%20Handle%20Structure))
  ;[0.006313 0.09822 0.15 0.5] Activate the ability if the proximity duration to the target (giant pawn) is beyond a desired time limit.
  (Utilities|FlowControl|Sequence
    (:then_0
      (bind _asbga_tidal_crane_boss_toy_fox (Utilities|Casting|CastToBGA_TidalCrane_Boss_ToyFox (Utilities|Struct|BreakGameplayAbilityActorInfo ActorInfo)))
      (bind _currenttargetactor (Class|BGATidalCraneBoss|GetCurrentTarget _asbga_tidal_crane_boss_toy_fox :!class /TidalCrane_Swarm_Boss/Gameplay/BGA_TidalCrane_Boss.BGA_TidalCrane_Boss_C))
      (Utilities|IsValid _currenttargetactor
        (:Is%20Valid
          (if (Math|Boolean|ORBoolean (>= (TargetProximity|GetTargetProximityDuration _asbga_tidal_crane_boss_toy_fox) (ScalableFloat|GetCachedScalableFloatValue (Variables|Hotfix|GetHotfix_ProximityDurationBeforeReposition))) (GameplayTags|HasAnyMatchingGameplayTags _asbga_tidal_crane_boss_toy_fox (GameplayTags|MakeLiteralGameplayTagContainer "(GameplayTags=((TagName=\"Ability.TidalCrane.ToyFox.Melee.HeadbuttRecentlyUsed\"),(TagName=\"Ability.TIdalCrane.ToyFox.Stunned.Recently\")))")) (>= (Frustration|GetFrustration _asbga_tidal_crane_boss_toy_fox) (ScalableFloat|GetCachedScalableFloatValue (Variables|Hotfix|GetHotfix_FrustrationPercentBeforeReposition))))
            _))
        (:Is%20Not%20Valid
          _))
      (return 0 true))
    (:then_1
      (return 0 false))))
